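# Sound cue for the Tinfoil Drill prop's ping, layered into close, medium and distant perspectives after a short fixed delay.
# The close branch mixes the main ping, a close layer and a further-delayed whoosh through overridden explosion-close attenuation.
# Medium reuses the main ping and distant plays the "Medium" wave, each with projectile attenuation.
# Each branch sets its volume and a ranged-weapon fire sound class; the cue overrides concurrency to cap voices.
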SoundCue /FNGameplayCues/Athena/Sounds/Props/Tinfoil_Drill/Tinfoil_Drill_Ping_Cue
root=Delay_0
nodes [
 {
  id=Attenuation_0
  children [WavePlayer_5]
  properties {
   attenuationSettings=&Fort_Weapon_Projectile_Distant_Mono_Athena_Att
   attenuationOverrides {falloffDistance=3000, attenuationShapeExtents {x=1000}}
  }
 }
 {
  id=Attenuation_4
  children [Mixer_1]
  properties {
   attenuationSettings=/Game/Sounds/Attenuation/Fort_Explosion_Close_Athena_Att
   bOverrideAttenuation=true
   attenuationOverrides {falloffDistance=4000, stereoSpread=5084.1587, attenuationShapeExtents {x=1000}}
  }
 }
 {
  id=Attenuation_5
  children [WavePlayer_17]
  properties {
   attenuationSettings=&Fort_Weapon_Projectile_Medium_Mono_Athena_Att
   attenuationOverrides {falloffDistance=3000, attenuationShapeExtents {x=1000}}
  }
 }
 {id=Delay_0, children [Mixer_4], properties {delayMax=0.2, delayMin=0.2}}
 {id=Delay_2, children [WavePlayer_4], properties {delayMax=0.3, delayMin=0.3}}
 {id=Mixer_1, children [WavePlayer_1,WavePlayer_6,Delay_2], properties {inputVolume [0.9,0.6,0.35]}}
 {id=Mixer_4, children [Modulator_4,Modulator_6,Modulator_7], properties {inputVolume [1,0.8,0.5]}}
 {id=Modulator_4, children [SoundClass_3], properties {pitchMax=1, pitchMin=1, volumeMax=0.95}}
 {id=Modulator_6, children [SoundClass_2], properties {pitchMax=1, pitchMin=1, volumeMax=0.5, volumeMin=0.5}}
 {id=Modulator_7, children [SoundClass_1], properties {pitchMax=1, pitchMin=1, volumeMax=0.6, volumeMin=0.6}}
 {
  id=SoundClass_1
  children [Attenuation_0]
  properties {soundClassOverride=&Fort_WeaponRanged_Fire_Distant}
 }
 {
  id=SoundClass_2
  children [Attenuation_5]
  properties {soundClassOverride=&Fort_WeaponRanged_Fire_Distant}
 }
 {
  id=SoundClass_3
  children [Attenuation_4]
  properties {soundClassOverride=&Fort_WeaponRanged_Fire_3P_PSM}
 }
 {id=WavePlayer_1, properties {soundWaveAssetPtr=&/Tinfoil_Drill_Ping_01}}
 {id=WavePlayer_17, properties {soundWaveAssetPtr=&Tinfoil_Drill_Ping_01}}
 {id=WavePlayer_4, properties {soundWaveAssetPtr=&/Tinfoil_Drill_Ping_Whoosh_01}}
 {id=WavePlayer_5, properties {soundWaveAssetPtr=&/Tinfoil_Drill_Ping_Medium_01}}
 {id=WavePlayer_6, properties {soundWaveAssetPtr=&/Tinfoil_Drill_Ping_CloseLayer_01}}
]
properties {bOverrideConcurrency=true, maxDistance=27500, volumeMultiplier=1, concurrencyOverrides {maxCount=10, volumeScale=0.8762}}
